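# Translucent, two-sided lit smoke material for the Krampus glider FX, with volumetric per-vertex directional lighting.
# Three speed-driven panning swirl flowmap samples combine into streaky noise, and two of them also drive a ColorA-to-ColorB blend along V.
# Opacity math multiplies that noise by a sphere mask, depth fade, inverted Fresnel and Wipe gradient.
# Panning noise weighted by inverse V scales the vertex normal by DistortAmount, apparently for vertex displacement.
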Material /Game/Gadgets/Assets/VinderTech_GliderChute/Glider_Krampus/FX/M_SwirlyColorfulSmoke_Cloudy
graph {
 properties {
  bUsedWithMeshDeformer=true
  bUsedWithSkeletalMesh=true
  blendMode=BLEND_Translucent
  pixelDepthOffsetMode=PDOM_Legacy
  translucencyLightingMode=TLM_VolumetricPerVertexDirectional
  twoSided=true
  usedShadingModels=Default Lit
 }
 nodes [
  n1 = Multiply(A=Time(), B=ScalarParameter() {defaultValue=2, parameterName=Speed})
  n2 = TextureSample(UVs=Panner(Coordinate=TextureCoordinate() {vTiling=0.5}, Time=@n1) {speedX=0.01, speedY=1}) {samplerType=SAMPLERTYPE_LinearColor, texture=/Game/Global/Textures/Default/Flowmaps/T_SwirlFlow_02}
  n3 = TextureSample(UVs=Panner(Coordinate=TextureCoordinate() {uTiling=3, vTiling=0.75}, Time=@n1) {speedX=-0.05, speedY=1}) {samplerType=SAMPLERTYPE_LinearColor, texture=&T_SwirlFlow_02}
  Add(A=@n2.G, B=@n3.G)
  n4 = Multiply(A=@n2.G, B=@n3.G)
  n5 = Add(A=Add(A=TextureCoordinate(), B=Constant2Vector() {g=0.25}), B=Multiply(A=@n4) {constB=0.5})
  LinearInterpolate(
   A=VectorParameter() {parameterName=ColorA, defaultValue=<rgba 0.1637 1.2404 2 1>}.RGB
   B=VectorParameter() {parameterName=ColorB, defaultValue=<rgba 0.7835 0.0168 0.0307 1>}.RGB
   Alpha=Power(Base=ComponentMask(None=@n5) {g=true}) {constExponent=4}
  )
  n6 = Constant3Vector() {constant=<rgba 0 1 0 0>}
  Multiply(
   A=OneMinus(None=MaterialFunctionCall() {materialFunction=&LinearGradient}.VGradient)
   B=Multiply(A=Transform(None=@n6) {transformSourceType=TRANSFORMSOURCE_World, transformType=TRANSFORM_Local})
  )
  n7 = Fresnel() {exponent=1}
  n8 = Multiply(
   A=Power(Base=OneMinus(None=@n7), Exp=ScalarParameter() {defaultValue=1.25, parameterName=FresnelFalloff})
   B=ScalarParameter() {defaultValue=0.8, parameterName=FresnelTightness}
  )
  n9 = Clamp(None=Add(
    A=MaterialFunctionCall() {materialFunction=&LinearGradient}.VGradient
    B=ScalarParameter() {defaultValue=1, parameterName=Wipe}
   ))
  n10 = ComponentMask(None=TextureCoordinate()) {g=true}
  n11 = Power(Base=SphereMask(A=@n10, B=Constant2Vector() {g=0.6, r=0.5}) {attenuationRadius=0.65, hardnessPercent=0}) {constExponent=3}
  n12 = TextureSample(UVs=Panner(Coordinate=TextureCoordinate() {uTiling=5, vTiling=1.5}, Time=@n1) {speedX=0.025, speedY=1}) {samplerType=SAMPLERTYPE_LinearColor, texture=&T_SwirlFlow_02}
  n13 = Power(Base=Clamp(None=Add(A=@n4, B=@n12.G)), Exp=LinearInterpolate(Alpha=Power(Base=@n10) {constExponent=1.5}) {constA=8, constB=0.05})
  n14 = DepthFade(Opacity=Clamp(None=Multiply(A=@n11, B=@n13)), FadeDistance=ScalarParameter() {defaultValue=3, parameterName=DepthFade})
  Multiply(A=Multiply(A=Clamp(None=@n8), B=Multiply(A=@n9, B=@n14)), B=ScalarParameter() {defaultValue=3, parameterName=OpacityBoost})
  n15 = TextureCoordinate()
  n16 = Multiply(
   A=Power(Base=OneMinus(None=@n10)) {constExponent=1}
   B=TextureSample(UVs=Panner(Coordinate=@n15, Time=@n1) {speedX=0.01, speedY=1}) {texture=&T_NoisePack_01}.G
  )
  n17 = ScalarParameter() {defaultValue=30, parameterName=DistortAmount}
  Reroute(None=Multiply(A=Multiply(A=@n16, B=@n17), B=VertexNormalWS()))
  TextureSample() {samplerType=SAMPLERTYPE_LinearColor, texture=&T_SwirlFlow_02}
 ]
}
